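# Conversation database for the NPC Crustina, defining the greeting shown when a player talks to her.
# A single entry point feeds one speech node with a default line plus contextual lines for visiting Durrr Burger, a first-ever conversation, or wearing specific rival or pizza-themed outfits.
# Curve table rows set the priority of each contextual line.
# The greeting then branches through link nodes to the shared NPC Library quests and basic services conversations.
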ConversationDatabase /NPCLibrary/NPCs/Crustina/Crustina/NPCLibrary_Conversation_Crustina
entry_points [{tag=Conversation.Entry.Crustina, destinations [#1]}]
nodes [
 ConversationEntryPointNode_7 {
  id=#1
  class=ConversationEntryPointNode
  outputs [#2]
  properties {entryTag {tagName=Conversation.Entry.Crustina}}
 }
 Speech_C_4 {
  id=#2
  class=Speech_C
  outputs [#3, #4]
  properties {
   nodeName=""
   generalConfig {
    defaultMessage=Keepin' it fresh?
    contextualMessages [
     {
      message="Ugh. You reek of *sniff sniff* <enemy>patty-stink</>. Better hope the boss doesn't find out you went there."
      contextRequirements [{requirement {class=FortControllerRequirement_HasVisitedPOI, properties {pOITag {tagName=Athena.Location.UnnamedPOI.DurrrBurger}}}}]
      priority {curve {rowName=Tandem.Speech.Priority.VisitedSpecificPOI}}
     }
     {
      message="Haven't seen you before. Don't cause any trouble and there won't be any trouble, mmkay?"
      contextRequirements [{requirement {class=FortControllerRequirement_IsFirstEverConversationWithNPC}}]
      priority {curve {rowName=Tandem.Speech.Priority.FirstEverInteraction}}
     }
     {
      message="Don't start nothing, won't be nothing."
      requirementMatchPolicy=RequireAny
      contextRequirements [{requirement {
         class=FortControllerRequirement_HasCID
         properties {allowedCIds [
           /BRCosmetics/Athena/Items/Cosmetics/Characters/CID_119_Athena_Commando_F_Candy
           /BRCosmetics/Athena/Items/Cosmetics/Characters/CID_184_Athena_Commando_M_DurrburgerWorker
           /BRCosmetics/Athena/Items/Cosmetics/Characters/CID_185_Athena_Commando_M_DurrburgerHero
           /BRCosmetics/Athena/Items/Cosmetics/Characters/CID_271_Athena_Commando_F_SushiChef
          ]}
        }}]
      priority {curve {rowName=Tandem.Speech.Priority.OutfitReaction}}
     }
     {
      message="Just say the word, boss."
      requirementMatchPolicy=RequireAny
      contextRequirements [{requirement {
         class=FortControllerRequirement_HasCID
         properties {allowedCIds [
           /BRCosmetics/Athena/Items/Cosmetics/Characters/CID_109_Athena_Commando_M_Pizza
           /BRCosmetics/Athena/Items/Cosmetics/Characters/CID_232_Athena_Commando_F_HalloweenTomato
          ]}
        }}]
      priority {curve {rowName=Tandem.Speech.Priority.OutfitReaction}}
     }
    ]
   }
  }
 }
 ConversationLinkNode_0 {id=#4, class=ConversationLinkNode, properties {remoteEntryTag {tagName=Conversation.Entry.BasicServices}}}
 ConversationLinkNode_1 {id=#3, class=ConversationLinkNode, properties {remoteEntryTag {tagName=Conversation.Entry.NPCLibrary.Quests}}}
]
